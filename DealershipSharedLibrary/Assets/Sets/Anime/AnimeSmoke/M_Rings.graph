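# Unlit, two-sided masked material for Niagara mesh particles, drawing stylized anime smoke rings.
# Dynamic parameters drive a vertex offset combining normal-based thickness with animated sine wobble from time-shifted world position.
# A cloud normal map dotted with up is smoothstepped into a toon shadow band, added to Shadow Color and multiplied by particle color for emissive.
# Cell noise offset by particle alpha forms the dissolve mask, multiplied by camera depth fade.
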Material /DealershipSharedLibrary/Assets/Sets/Anime/AnimeSmoke/M_Rings
graph {
 properties {
  bCastRayTracedShadows=false
  bUsedWithNiagaraMeshParticles=true
  blendMode=BLEND_Masked
  shadingModel=MSM_Unlit
  twoSided=true
  usedShadingModels=Unlit
 }
 nodes [
  n1 = DynamicParameter() {defaultValue=<rgba 1 1 1 0.3>, paramNames [Thickness, Param2, Noise Strengh, Noise Frequency]}
  n2 = Add(A=Multiply(A=Constant() {r=0.05}, B=Time()), B=WorldPosition().XYZ)
  n3 = ComponentMask(None=@n2) {r=true}
  n4 = NamedRerouteUsage()
  n5 = ComponentMask(None=@n2) {g=true}
  n6 = NamedRerouteUsage()
  n7 = AppendVector(A=Sine(None=Multiply(A=@n3, B=@n4)), B=Sine(None=Multiply(A=@n5, B=@n6)))
  n8 = Sine(None=Multiply(A=ComponentMask(None=@n2) {b=true}, B=NamedRerouteUsage()))
  Add(A=Multiply(A=VertexNormalWS(), B=@n1.Thickness), B=Multiply(A=@n1.Noise Strengh, B=AppendVector(A=@n7, B=@n8)))
  n9 = ScalarParameter() {defaultValue=0.5, group=Shadow, parameterName=ScaleShadow X}
  n10 = ScalarParameter() {defaultValue=0.1, group=Shadow, parameterName=ScaleShadow Y}
  n11 = TextureSampleParameter2D(UVs=Multiply(A=TextureCoordinate(), B=AppendVector(A=@n9, B=@n10))) {group=Shadow, parameterName=Shadow Text, samplerType=SAMPLERTYPE_Normal, texture=/DealershipSharedLibrary/FX/Textures/Normal/T_CloudDetail_N}
  n12 = DotProduct(A=Normalize(VectorInput=Constant3Vector() {constant=<rgba 0 0 1 0>}), B=Transform(None=@n11.RGB))
  n13 = Add(A=ScalarParameter() {defaultValue=-1.5, group=Shadow, parameterName=Shadow Height}, B=Constant() {r=1})
  n14 = SmoothStep(Min=Constant() {r=0.1}, Max=Constant() {r=0.25}, Value=Saturate(None=Add(A=@n12, B=@n13)))
  n15 = VectorParameter() {group=Shadow, parameterName=Shadow Color, defaultValue=<rgba 0.1481 0.4016 0.474 1>}
  n16 = ParticleColor()
  MaterialFunctionCall(Emissive=Multiply(A=Add(A=@n14, B=@n15.RGB), B=@n16.RGB)) {materialFunction=&MF_FX_EmissiveAdjustments}
  n17 = ScalarParameter() {defaultValue=0.75, group=Dissolve, parameterName=DissolveScale X}
  n18 = ScalarParameter() {defaultValue=1.5, group=Dissolve, parameterName=DissolveScale Y}
  n19 = TextureSampleParameter2D(UVs=Multiply(A=TextureCoordinate(), B=AppendVector(A=@n17, B=@n18))) {
   group=Dissolve
   parameterName=Dissolve Text
   texture=/Game/Environments/Helios/Props/Quest/Quest_GoldenFleeceStatuette_A/Textures/T_GoldenFleeceStatuette_CellNoise
  }
  n20 = Constant() {r=1}
  Multiply(
   A=Subtract(A=Add(A=@n19.R, B=@n20), B=@n16.A)
   B=MaterialFunctionCall(Fade Length=Constant() {r=300}, Fade Offset=Constant() {r=100}) {materialFunction=&CameraDepthFade}.Result
  )
  NamedRerouteDeclaration(None=@n1.Noise Frequency) {name=Noise Frequency, nodeColor=<rgba 0 0.6016 1 1>}
 ]
}
