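# Translucent, two-sided Niagara material for an anime-style shockwave sphere with glowing, noisy Fresnel edges.
# Panned noise, edge Fresnel and optional core terms combine through contrast into tinted emissive; a dynamic Dissolve parameter can drive their power exponents.
# Base color mixes particle tint with optional desaturated scene color, falling back to black on low-end feature levels.
# Opacity adds inverse-Fresnel fill, optional particle alpha and switchable depth fade on deferred SM5/SM6.
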Material /DealershipSharedLibrary/Assets/Sets/Anime/M_DLR_Shockwave_Sphere
graph {
 properties {
  bAutomaticallySetUsageInEditor=false
  bUsedWithNiagaraMeshParticles=true
  bUsedWithNiagaraRibbons=true
  bUsedWithNiagaraSprites=true
  blendMode=BLEND_Translucent
  twoSided=true
  usedShadingModels=Default Lit
 }
 nodes [
  n1 = Constant()
  n2 = Desaturation(
   None=FeatureLevelSwitch(Default=SceneColor!(UVs).RGB, ES2=@n1, ES3_1=@n1, SM4_REMOVED=@n1)
   Fraction=ScalarParameter() {defaultValue=1, group=Look, parameterName=Desaturation Amount}
  )
  n3 = ScalarParameter() {defaultValue=0.2, group=Look, parameterName=Contrast}
  n4 = Constant()
  n5 = ParticleColor()
  n6 = ScalarParameter() {defaultValue=1, parameterName=Use Particle Color?}
  Add(
   A=If(
    A=ScalarParameter() {parameterName=Desaturation On/Off}
    "A > B"=MaterialFunctionCall(In=@n2, Contrast=@n3) {materialFunction=&CheapContrast}.Result
    "A == B"=@n4
    "A < B"=@n4
   )
   B=Multiply(A=VectorParameter() {group=Look, parameterName=Base Color, defaultValue=<rgba 0 0 0 1>}.RGB, B=Multiply(A=@n5.RGB, B=@n6))
  )
  MaterialFunctionCall(Emissive=NamedRerouteUsage()) {materialFunction=&MF_FX_EmissiveAdjustments}
  n7 = TextureCoordinate()
  n8 = AppendVector(
   A=ScalarParameter() {defaultValue=0.25, group=Noise, parameterName=Tile X}
   B=ScalarParameter() {defaultValue=10, group=Noise, parameterName=Tile Y}
  )
  n9 = ScalarParameter() {defaultValue=0.05, group=Noise, parameterName=Speed X}
  n10 = ScalarParameter() {group=Noise, parameterName=Speed Y}
  n11 = TextureSampleParameter2D(UVs=Panner(Coordinate=Multiply(A=@n7, B=@n8), Speed=AppendVector(A=@n9, B=@n10))) {parameterName=Noise Texture, texture=/Game/Effects/Fort_Effects/Textures/Tile/T_GaussianNoise_01_M}
  n12 = Multiply(
   A=ScalarParameter() {defaultValue=2, group=Noise, parameterName=Noise Amount}
   B=StaticSwitchParameter(True=NamedRerouteUsage(), False=Constant() {r=1}) {defaultValue=true, parameterName=Use Dyamic Dissolve?}
  )
  NamedRerouteDeclaration(None=Clamp(None=Power(Base=@n11.RGB, Exp=@n12) {constExponent=1})) {name=Noise_Out, nodeColor=<rgba 1 0.9721 0.3209 1>}
  NamedRerouteDeclaration(None=DynamicParameter() {paramNames [Dissolve, Noise Dissolve, Param3, Param4]}.Dissolve) {name=Dissolve, nodeColor=<rgba 1 0 0 1>}
  n13 = Fresnel() {exponent=1}
  n14 = Multiply(
   A=ScalarParameter() {defaultValue=4, group=Fresnel, parameterName=Noise Frensel Power}
   B=StaticSwitchParameter(True=NamedRerouteUsage(), False=Constant() {r=1}) {defaultValue=true, parameterName=Use Dyamic Dissolve?}
  )
  n15 = Multiply(
   A=NamedRerouteUsage()
   B=Multiply(A=Power(Base=@n13, Exp=@n14) {constExponent=5}, B=ScalarParameter() {defaultValue=1, group=Fresnel, parameterName=Edge?})
  )
  n16 = ScalarParameter() {defaultValue=5, group=Fresnel, parameterName=Edge Fresnel Exponent}
  n17 = ScalarParameter() {defaultValue=0.04, group=Fresnel, parameterName=Edge Fresnel Reflect}
  n18 = ScalarParameter() {defaultValue=4, group=Fresnel, parameterName=Edge Frensel Power}
  n19 = StaticSwitchParameter(True=NamedRerouteUsage(), False=Constant() {r=1}) {defaultValue=true, parameterName=Use Dyamic Dissolve?}
  n20 = Multiply(
   A=Power(Base=Fresnel(ExponentIn=@n16, BaseReflectFractionIn=@n17), Exp=Multiply(A=@n18, B=@n19)) {constExponent=5}
   B=ScalarParameter() {defaultValue=1, group=Fresnel, parameterName=Edge?}
  )
  n21 = Fresnel(
   ExponentIn=ScalarParameter() {defaultValue=5, group=Core, parameterName=Core Fresnel Exponent}
   BaseReflectFractionIn=ScalarParameter() {defaultValue=0.04, group=Core, parameterName=Core Fresnel Reflect}
  ) {exponent=0.5}
  n22 = Multiply(
   A=ScalarParameter() {defaultValue=0.25, group=Fresnel, parameterName=Core Frensel Power}
   B=StaticSwitchParameter(True=NamedRerouteUsage(), False=Constant() {r=1}) {defaultValue=true, parameterName=Use Dyamic Dissolve?}
  )
  n23 = Clamp(None=OneMinus(None=Power(Base=@n21, Exp=@n22) {constExponent=5}))
  n24 = NamedRerouteUsage()
  n25 = LinearInterpolate(
   B=LinearInterpolate(A=@n23, B=Multiply(A=@n23, B=@n24), Alpha=ScalarParameter() {group=Core, parameterName=Core Noise?})
   Alpha=ScalarParameter() {group=Core, parameterName=Core?}
  )
  n26 = Clamp(None=Add(A=Add(A=@n15, B=@n20), B=Clamp(None=@n25)))
  n27 = ScalarParameter() {defaultValue=0.05, group=Look, parameterName=Edge Hardness}
  NamedRerouteDeclaration(None=Multiply(
    A=MaterialFunctionCall(In=@n26, Contrast=@n27) {materialFunction=&CheapContrast}.Result
    B=ScalarParameter() {defaultValue=10, group=Look, parameterName=Emissive Boost}
   )) {name=Fresnel_Out, nodeColor=<rgba 0.273 1 0.418 1>}
  n28 = VectorParameter() {group=Look, parameterName=Emissive Color, defaultValue=<rgba 1 1 1 0>}
  n29 = ParticleColor()
  NamedRerouteDeclaration(None=Multiply(A=Multiply(A=@n28.RGB, B=@n29.RGB), B=NamedRerouteUsage())) {name=Emissive Out, nodeColor=<rgba 1 0.1558 0.8187 1>}
  n30 = ParticleColor()
  n31 = NamedRerouteUsage()
  n32 = ScalarParameter() {group=Look, parameterName=Use Particle Alpha?}
  n33 = Clamp(None=OneMinus(None=Fresnel()))
  n34 = ScalarParameter() {defaultValue=1, group=Look, parameterName=Opacity Fill}
  n35 = Add(
   A=LinearInterpolate(
    A=@n31
    B=Multiply(A=@n30.A, B=@n31)
    Alpha=ScalarParameter() {desc=Added to allow fade out of edge - Jocelyn M, group=Look, parameterName=Use Particle Alpha on Edge?}
   ) {constA=1}
   B=Multiply(A=LinearInterpolate(B=@n30.A, Alpha=@n32) {constA=1}, B=Multiply(A=@n33, B=@n34))
  )
  n36 = LinearInterpolate(
   A=ScalarParameter() {defaultValue=800, group=Opacity, parameterName=Opacity - DepthFade - Min}
   B=ScalarParameter() {defaultValue=100, group=Opacity, parameterName=Opacity - DepthFade - Max}
   Alpha=@n35
  )
  n37 = ScalarParameter() {defaultValue=1, parameterName=ScaleFactor}
  n38 = ShadingPathSwitch(Default=@n35, Deferred=DepthFade(Opacity=@n35, FadeDistance=Multiply(A=@n36, B=@n37))) {desc="Defered is pc, Forward is switch, mobile is mobile"}
  n39 = FeatureLevelSwitch(Default=@n35, ES2=@n35, ES3_1=@n35, SM4_REMOVED=@n35, SM5=@n38, SM6=@n38)
  n40 = Constant() {r=1}
  NamedRerouteDeclaration(None=Multiply(A=@n35, B=StaticSwitchParameter(True=@n39, False=@n40) {group=Opacity, parameterName=UseDephtFade, sortPriority=10})) {name=Opacity Out, nodeColor=<rgba 0 0.7557 1 1>}
  NamedRerouteUsage()
 ]
}
