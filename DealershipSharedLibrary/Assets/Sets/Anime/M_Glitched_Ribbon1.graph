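# Translucent Niagara material for an anime-set "glitched" ribbon effect, also usable on sprite and mesh particles.
# Particle color, multiplied by two procedural band masks, forms the glow.
# A stripe texture builds opacity, with edge highlights from subtracting two contrast curves, plus optional depth fade on SM5/SM6 deferred.
# Layered time-driven sine/cosine waves form small ridges and large curves for world position offset, with tiling, speed and amount set by dynamic parameters.
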Material /DealershipSharedLibrary/Assets/Sets/Anime/M_Glitched_Ribbon1
graph {
 properties {
  bAutomaticallySetUsageInEditor=false
  bUsedWithNiagaraMeshParticles=true
  bUsedWithNiagaraRibbons=true
  bUsedWithNiagaraSprites=true
  blendMode=BLEND_Translucent
  usedShadingModels=Default Lit
 }
 nodes [
  n1 = TextureCoordinate()
  Clamp(None=Multiply(A=ComponentMask(None=@n1) {r=true}) {constB=10}) {minDefault=0.5}
  n2 = ParticleColor()
  n3 = ScalarParameter() {defaultValue=0.5, group=Panning Band, parameterName=Panning Band Width, sortPriority=3}
  n4 = ScalarParameter() {group=Panning Band, parameterName="Panning  Band Sharpness", sortPriority=3}
  n5 = StaticBool() {value=true}
  n6 = Multiply(A=TextureCoordinate(), B=ScalarParameter() {defaultValue=7, group=Panning Band, parameterName="Panning  Band Tiling", sortPriority=2})
  n7 = Clamp(None=Power(Base=MaterialFunctionCall!(Width=@n3, Sharpness=@n4, Offset, Direction Switch=@n5, Compare, Input Coordinates=@n6) {materialFunction=/Engine/Functions/Engine_MaterialFunctions03/Procedurals/GeneratedBand}.Result) {constExponent=1}) {minDefault=0.3}
  n8 = MaterialFunctionCall(
   Width=ScalarParameter() {defaultValue=0.2, group=Panning Band, parameterName=Band Width, sortPriority=3}
   Sharpness=ScalarParameter() {group=Panning Band, parameterName=Band Sharpness, sortPriority=3}
   Offset=Constant() {r=-0.03}
   Direction Switch=StaticBool()
   Input Coordinates=Multiply(A=TextureCoordinate(), B=ScalarParameter() {defaultValue=1, group=Panning Band, parameterName=Band Tiling, sortPriority=2})
  ) {materialFunction=&GeneratedBand}
  n9 = Multiply(A=Clamp(None=Power(Base=@n8.Result) {constExponent=7}) {minDefault=0.1}) {constB=5}
  Multiply(
   A=Multiply(
    A=MaterialFunctionCall(Emissive=@n2.RGB) {materialFunction=&MF_FX_EmissiveAdjustments}.Result
    B=Multiply(A=@n7, B=@n9)
   )
   B=ScalarParameter() {defaultValue=0.65, parameterName=Emissive Gain}
  )
  n10 = Multiply(A=ComponentMask(None=TextureCoordinate()) {r=true}) {constB=10}
  n11 = OneMinus(None=Time())
  n12 = Add(A=ComponentMask(None=TextureCoordinate() {uTiling=8, vTiling=8}) {r=true}, B=Multiply(A=Multiply(A=@n11) {constB=20}, B=NamedRerouteUsage()))
  n13 = Sine(None=@n12)
  n14 = AppendVector(A=Cosine(None=@n12), B=Multiply(A=@n13, B=VertexColor().A))
  n15 = Multiply(A=Transform(None=AppendVector(A=@n14, B=@n13))) {constB=40}
  n16 = ComponentMask(None=Subtract(A=TextureCoordinate()) {constB=0.475}) {g=true}
  n17 = Multiply(
   A=Multiply(A=Clamp(None=@n10) {minDefault=0.5}, B=Multiply(A=@n15, B=@n16))
   B=ScalarParameter() {defaultValue=0.6, parameterName=Small Ridge Amount}
  ) {constB=0.6}
  n18 = TextureCoordinate()
  n19 = Multiply(A=Constant() {r=0.25}, B=NamedRerouteUsage())
  n20 = ComponentMask(None=Add(A=Multiply(A=@n18, B=@n19), B=ScalarParameter() {parameterName=Rotation Offset})) {r=true}
  n21 = Time()
  n22 = Multiply(A=Constant() {r=1.25}, B=Multiply(A=OneMinus(None=@n21), B=NamedRerouteUsage()))
  n23 = Add(A=@n20, B=@n22)
  n24 = Multiply(A=@n22) {constB=0.5}
  n25 = Add(A=Add(A=Cosine(None=@n23), B=Cosine(None=@n23) {period=2}), B=Cosine(None=Add(A=@n24, B=@n20)) {period=0.75})
  n26 = Add(A=@n20, B=@n22)
  n27 = Sine(None=@n26)
  n28 = Sine(None=@n26) {period=2}
  n29 = AppendVector(A=@n25, B=Add(A=Add(A=@n27, B=@n28), B=Sine(None=@n26) {period=0.75}))
  n30 = Multiply(A=Transform(None=AppendVector(A=@n29, B=@n25)), B=ScalarParameter() {defaultValue=2.5, parameterName=Large Curve Amount}) {constB=2.5}
  NamedRerouteDeclaration(None=Multiply(A=Add(A=@n17, B=@n30), B=NamedRerouteUsage())) {name=WPO, nodeColor=<rgba 1 0.7777 0.151 1>}
  n31 = DynamicParameter() {parameterIndex=1, paramNames [Curve Tiling, Speed, WPOAmount, Param4]}
  NamedRerouteDeclaration(None=@n31.Curve Tiling) {name=Curve Tiling, nodeColor=<rgba 1 0 0 1>}
  n32 = TextureSampleParameter2D(UVs=TextureCoordinate() {uTiling=10, unMirrorU=true}) {parameterName=Base Texture, texture=&/GradientStripe_T_anime}
  n33 = Subtract(
   A=MaterialFunctionCall(In=@n32.R, Contrast=Constant() {r=3.2}) {materialFunction=&CheapContrast}.Result
   B=MaterialFunctionCall(In=@n32.R, Contrast=Constant() {r=3}) {materialFunction=&CheapContrast}.Result
  )
  n34 = ScalarParameter() {defaultValue=6, parameterName=Edge Highlight Opacity}
  n35 = Power(Base=@n32.R) {constExponent=7}
  n36 = ScalarParameter() {defaultValue=0.5, parameterName=Base Gradient Opacity}
  n37 = Clamp(None=Add(A=Multiply(A=@n33, B=@n34), B=Multiply(A=@n35, B=@n36) {constB=0.5}))
  NamedRerouteDeclaration(None=@n37) {name=TextureOut, nodeColor=<rgba 0.4531 1 0 1>}
  NamedRerouteDeclaration(None=@n31.WPOAmount) {name=wpo Amount, nodeColor=<rgba 0 0.0746 1 1>}
  n38 = Multiply(A=ParticleColor().A, B=@n37)
  n39 = LinearInterpolate(
   A=ScalarParameter() {defaultValue=800, group=Opacity, parameterName=Opacity - DepthFade - Min}
   B=ScalarParameter() {defaultValue=100, group=Opacity, parameterName=Opacity - DepthFade - Max}
   Alpha=@n38
  )
  n40 = ScalarParameter() {defaultValue=1, parameterName=ScaleFactor}
  n41 = ShadingPathSwitch(Default=@n38, Deferred=DepthFade(Opacity=@n38, FadeDistance=Multiply(A=@n39, B=@n40))) {desc="Defered is pc, Forward is switch, mobile is mobile"}
  n42 = FeatureLevelSwitch(Default=@n38, ES2=@n38, ES3_1=@n38, SM4_REMOVED=@n38, SM5=@n41, SM6=@n41)
  n43 = Constant() {r=1}
  NamedRerouteDeclaration(None=Multiply(A=@n38, B=StaticSwitchParameter(True=@n42, False=@n43) {group=Opacity, parameterName=UseDephtFade, sortPriority=10})) {name=Opacity, nodeColor=<rgba 0.1138 0.9471 1 1>}
  NamedRerouteDeclaration(None=@n31.Speed) {name=Speed, nodeColor=<rgba 0.5469 1 0 1>}
  NamedRerouteUsage()
  n44 = Multiply(A=NamedRerouteUsage(), B=@n2.A)
  Reroute(None=Multiply(A=Clamp(None=@n44), B=ScalarParameter() {defaultValue=2.5, parameterName=Opacity Gain}))
 ]
}
