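# Translucent Niagara ribbon material for the Anime set, producing an animated, wavy, glowing ribbon.
# World position offset sums time-scrolled sine/cosine waves along UVs, one layered and one varying across ribbon width, pinned near U=0.
# Particle dynamic parameters scale both waves, add a separate offset and blend particle color with a default.
# A noise texture shapes emissive brightness through MF_FX_EmissiveAdjustments and, multiplied by particle alpha, opacity.
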Material /DealershipSharedLibrary/Assets/Sets/Anime/Materials/M_DLR_distorted_ribbon
graph {
 properties {bAutomaticallySetUsageInEditor=false, bUsedWithNiagaraRibbons=true, blendMode=BLEND_Translucent, usedShadingModels=Default Lit}
 nodes [
  n1 = ComponentMask(None=TextureCoordinate()) {r=true}
  n2 = Time()
  n3 = Add(
   A=ComponentMask(None=TextureCoordinate()) {r=true}
   B=Multiply(A=OneMinus(None=@n2), B=ScalarParameter() {defaultValue=20, parameterName=Curve 2 Speed}) {constB=16}
  )
  n4 = Sine(None=@n3)
  n5 = ScalarParameter() {defaultValue=0.5, parameterName=Curve 2 Height}
  n6 = AppendVector(A=AppendVector(A=Cosine(None=@n3), B=Multiply(A=@n4, B=@n5)), B=@n4)
  n7 = TextureCoordinate()
  n8 = Multiply(
   A=Multiply(A=Transform(None=@n6), B=ScalarParameter() {defaultValue=90, parameterName=Curve 2 Amount}) {constB=8}
   B=ComponentMask(None=Subtract(A=@n7) {constB=0.5}) {g=true}
  )
  n9 = DynamicParameter() {defaultValue=<rgba 0 1 30 1>, paramNames [Color Curve, Intensity, Curve 1 Amount, Curve 2 Amount]}
  n10 = ComponentMask(None=TextureCoordinate() {coordinateIndex=1}) {r=true}
  n11 = Multiply(A=ScalarParameter() {defaultValue=1, parameterName=Curve 1 Speed}, B=OneMinus(None=Time())) {constA=1}
  n12 = Add(A=@n11, B=@n10)
  n13 = Multiply(A=@n11) {constB=0.5}
  n14 = Add(A=Add(A=Cosine(None=@n12), B=Cosine(None=@n12) {period=2}), B=Cosine(None=Add(A=@n13, B=@n10)) {period=0.75})
  n15 = Add(A=@n11, B=@n10)
  n16 = Add(A=Add(A=Sine(None=@n15), B=Sine(None=@n15) {period=2}), B=Sine(None=@n15) {period=0.75})
  n17 = Transform(None=AppendVector(A=AppendVector(A=@n14, B=@n16), B=@n14))
  n18 = DynamicParameter() {defaultValue=<rgba 0 1 30 1>, paramNames [Color Curve, Intensity, Curve 1 Amount, Curve 2 Amount]}
  n19 = Multiply(
   A=Clamp(None=Multiply(A=@n1) {constB=16})
   B=Add(A=Multiply(A=@n8, B=@n9.Curve 2 Amount), B=Multiply(A=@n17, B=@n18.Curve 1 Amount) {constB=4})
  )
  n20 = DynamicParameter() {defaultValue=<rgba 0 1 30 1>, paramNames [Color Curve, Intensity, Curve 1 Amount, Curve 2 Amount]}
  n21 = DynamicParameter() {parameterIndex=1, paramNames [ColorCurve,Param2,Param3,Param4]}
  n22 = AppendVector(A=AppendVector(A=@n21.ColorCurve, B=@n21.ColorCurve), B=@n21.Param3)
  Add(
   A=Multiply(A=Multiply(A=@n19, B=@n20.Intensity), B=ScalarParameter() {defaultValue=1, parameterName=Local WPO Amount})
   B=Multiply(A=Transform(None=@n22), B=ScalarParameter() {parameterName=World WPO Amount})
  )
  n23 = LinearInterpolate(
   A=ParticleColor().RGB
   B=VectorParameter() {parameterName=DefaultColor, defaultValue=<rgba 0.147 0.0999 1 0.5059>}.RGB
   Alpha=DynamicParameter() {defaultValue=<rgba 0 1 30 1>, paramNames [Color Curve, Intensity, Curve 1 Amount, Curve 2 Amount]}.Color Curve
  )
  n24 = TextureCoordinate() {uTiling=0.25}
  n25 = Clamp(None=Add(
    A=TextureSample(UVs=@n24) {texture=/DealershipSharedLibrary/Assets/Sets/Anime/Textures/Noisel_RGB_T}.B
    B=ScalarParameter() {parameterName=Texture Bias}
   ))
  n26 = ScalarParameter() {defaultValue=3, parameterName=Emissive Color Mask}
  n27 = LinearInterpolate(
   A=Multiply(A=@n23, B=ScalarParameter() {defaultValue=0.1, parameterName=Base Color Value})
   B=@n23
   Alpha=Add(A=Power(Base=@n25) {constExponent=1}, B=Power(Base=@n25, Exp=@n26))
  )
  n28 = ScalarParameter() {defaultValue=0.1, parameterName=Emissive Color Add}
  MaterialFunctionCall(Emissive=Multiply(A=Add(A=@n27, B=@n28) {constB=0.1}, B=ScalarParameter() {defaultValue=1.75, parameterName=Emissive Color Value})) {materialFunction=&MF_FX_EmissiveAdjustments}
  Multiply(A=@n25, B=ParticleColor().A)
  ScalarParameter() {defaultValue=30, parameterName=Curve 1 Amount}
 ]
}
